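# Niagara system for the Lone Wolf Male glider's trails, drawn by one ribbon emitter, SideWings.
# It spawns two anim trails between bone or socket pairs on the user skeletal mesh, pushed along Y with short-lived particles.
# User.ColorIndex selects red, blue or green, and normalized age overrides ribbon U0.
# A system integrator value drives gameplay alpha, with a separate front-end value through a modifier.
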NiagaraSystem /BRCosmetics/Gadgets/Assets/VinderTech_GliderChute/Glider_Lone_Wolf_Male/FX/NS_Glider_Lone_Wolf_Male_Trails
dependencies {
 usedDataInterfaces []
 usedDynamicInputs []
 usedModules [&SystemState, &FN_Integrator]
}
emitters {SideWings {
  emitterSpawnScript [&FN_RibbonIDOffsetUtility()]
  emitterUpdateScript [
   &EmitterState(Delay First Loop Only=true, Life Cycle Mode=::NewEnumerator1, Loop Delay=1, Loop Duration=1, UseLoopDelay=true)
   &FN_SpawnAnimTrail(Number of Anim Trails=2)
  ]
  particleSpawnScript [
   &FN_InitializeAnimTrailBoneIndecies()
   &FN_InitializeAnimTrailBoneIndecies(AnimTrail Selection=1, First Bone or Socket Index=2, Second Bone or Socket Index=3) {module=FN_InitializeAnimTrailBoneIndecies001}
   &FN_InitializeAnimTrail(
    Color=&self:3ColorSwitch(
     0:LinearColor=<rgba 0.9583 0.0349 0.0581 2>
     1:LinearColor=<rgba 0 0.3629 2 1>
     2:LinearColor=<rgba 0.1218 0.849 0.0398 1>
     Index:NiagaraInt32=&MakeIntFromFloat(Float=$User.ColorIndex)
    )
    Lifetime=0.2
   )
   &AddVelocity(
    Coordinate Space=::World
    Velocity=&RotateVectorByQuaternion(Vector To Rotate=<1000 0 0>, Quaternion=$Output.SkeletalMeshLocation.SampledRotation)
   ) {enabled=false}
   &DynamicMaterialParameters(
    Index 0 Param 1=$Particles.MaterialRandom
    Index 0 Param 2=$Particles.MaterialRandom
    Module.Param0WriteEnabled=true
    Module.Param1WriteEnabled=true
    Module.Param2WriteEnabled=true
    Module.Param3WriteEnabled=true
   ) {enabled=false}
   &AddVelocity(Velocity=<0 2500 0>) {module=AddVelocity001}
  ]
  particleUpdateScript [
   &SolveForcesAndVelocity(Clamp Velocity=false, Limit Acceleration=false)
   &FN_UpdateAnimTrail(SkeletalMesh=$User.SkeletalMesh) {module=FN_UpdateAnimTrail001}
   /SetVariables(Particles.RibbonU0Override=$Particles.NormalizedAge) {isSetParameters=true}
   &ScaleColor(
    Scale Alpha=&FN_FrontEndFloatModifier(Gameplay=&Lerp_Float(A=0, B=0.8, Alpha=$System.FN_Integrator.Value), Front End=1)
    ScaleA=true
    ScaleRGB=true
   ) {module=ScaleColor001}
  ]
  rendererClasses [&NiagaraRibbonRendererProperties]
  renderers [{rendererClass=&NiagaraRibbonRendererProperties}]
 }}
summary {
 systemName=NS_Glider_Lone_Wolf_Male_Trails
 emitters [{emitterName=SideWings, rendererClasses [&NiagaraRibbonRendererProperties]}]
 userVariables [
  User.bIsFrontEnd {type=NiagaraBool}
  User.bIsFrontEndPreview {type=NiagaraBool}
  User.bIsFullyDeployed {type=NiagaraBool}
  User.ColorIndex {type=NiagaraFloat, defaultValue=0, description=""}
  User.ForwardDot {type=NiagaraFloat, defaultValue=0, description=""}
  User.RibbonIDOffset {type=NiagaraInt32}
  User.RightDot {type=NiagaraFloat, defaultValue=0, description=""}
  User.SkeletalMesh {
   type=NiagaraDataInterfaceSkeletalMesh
   defaultValue {
    struct=NiagaraExt_VariableValue_DataInterface
    value {dataInterface=&self:NiagaraDataInterfaceSkeletalMesh_0}
   }
  }
 ]
}
